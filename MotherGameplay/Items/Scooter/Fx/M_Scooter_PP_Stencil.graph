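# Post-process material for the Scooter item that adds red and yellow highlights to objects tagged with the Scooter overlay stencil IDs.
# It builds a pulsing outline and fill from depth-tested custom-depth samples, masked by stencil compares.
# An expanding, fading ring driven by the Interrogate parameter collection's origin, radius and timestamp modulates brightness.
# A world-aligned grid scan with depth edge detection is also built but does not feed the final color switch.
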Material /MotherGameplay/Items/Scooter/Fx/M_Scooter_PP_Stencil
graph {
 properties {
  blendMode=BLEND_Translucent
  blendableLocation=BL_SceneColorBeforeDOF
  blendablePriority=-5
  materialDomain=MD_PostProcess
  pixelDepthOffsetMode=PDOM_Legacy
  shadingModel=MSM_Unlit
  usedShadingModels=Unlit
 }
 nodes [
  n1 = MaterialFunctionCall(
   TimestampA=CollectionParameter() {collection=/FNGameplayCues/Creative/Items/Gameplay/Interrogate/MPC_Interrogate, parameterName=ActivatedTimestamp}
   TransitionLength_In=CollectionParameter() {collection=&MPC_Interrogate, parameterName=Duration}
   Transition In Only?=StaticBool() {value=true}
  ) {materialFunction=&MF_Timestamp_InOut_01}
  n2 = MaterialFunctionCall(
   TextureObject=TextureObject() {samplerType=SAMPLERTYPE_LinearColor, texture=/Game/Gadgets/Assets/Teleporter/Textures/T_Teleporter_GridSquares_01}
   TextureSize=ScalarParameter() {defaultValue=256, parameterName=WorldTextureScale}
   World Space Normal=ShadingPathSwitch(Default=VertexNormalWS(), Deferred=SceneTexture() {sceneTextureId=PPI_WorldNormal}.Color)
  ) {materialFunction=&WorldAlignedTexture}
  n3 = Sine(None=Add(A=@n1.Linear, B=ComponentMask(None=@n2.XYZ Texture) {r=true}))
  n4 = SphereMask(
   A=ComponentMask(None=CollectionParameter() {collection=&MPC_Interrogate, parameterName=OriginPositionWS}) {b=true, g=true, r=true}
   B=WorldPosition().XYZ
   Radius=Multiply(A=CollectionParameter() {collection=&MPC_Interrogate, parameterName=Radius}, B=@n1.Linear)
   Hardness=Constant()
  )
  n5 = SphereMask(
   A=Power(Base=OneMinus(None=@n4), Exp=ScalarParameter() {defaultValue=1.5, parameterName=Exponent})
   B=ScalarParameter() {defaultValue=0.5, parameterName=RingCenter}
   Radius=ScalarParameter() {defaultValue=0.5, parameterName=RingRadius}
   Hardness=ScalarParameter() {defaultValue=0.2, parameterName=RingHardness}
  )
  n6 = Constant() {r=1}
  n7 = ScreenPosition()
  n8 = Multiply(B=SceneTexelSize()) {constA=3}
  n9 = Multiply(A=@n8, B=Constant2Vector() {g=-1, r=-1})
  n10 = SceneTexture() {sceneTextureId=PPI_SceneDepth}
  n11 = Subtract(A=SceneTexture(UVs=Add(A=@n7.ViewportUV, B=@n9)) {sceneTextureId=PPI_SceneDepth}.Color, B=@n10.Color)
  n12 = Multiply(A=@n8, B=Constant2Vector() {g=-1, r=1})
  n13 = Subtract(A=SceneTexture(UVs=Add(A=@n7.ViewportUV, B=@n12)) {sceneTextureId=PPI_SceneDepth}.Color, B=@n10.Color)
  n14 = Constant2Vector() {g=1, r=1}
  n15 = SceneTexture(UVs=Add(A=@n7.ViewportUV, B=Multiply(A=@n8, B=@n14))) {sceneTextureId=PPI_SceneDepth}
  n16 = Add(A=@n7.ViewportUV, B=Multiply(A=@n8, B=Constant2Vector() {g=1, r=-1}))
  n17 = Add(
   A=Add(A=Add(A=@n11, B=@n13), B=Subtract(A=@n15.Color, B=@n10.Color))
   B=Subtract(A=SceneTexture(UVs=@n16) {sceneTextureId=PPI_SceneDepth}.Color, B=@n10.Color)
  )
  n18 = Multiply(A=ComponentMask(None=Abs(None=@n17)) {r=true}, B=ScalarParameter() {defaultValue=0.01, parameterName=EdgeDetectionScaleA})
  n19 = Multiply(A=Round(None=Saturate(None=@n18)), B=ScalarParameter() {defaultValue=10, parameterName=EdgeDetectionScaleB})
  n20 = Add(A=@n5, B=Multiply(A=@n5, B=ShadingPathSwitch(Default=@n6, Deferred=@n19)))
  n21 = ComponentMask(None=SceneTexture() {sceneTextureId=PPI_PostProcessInput0}.Color) {b=true, g=true, r=true}
  n22 = Desaturation(None=@n21, Fraction=ScalarParameter() {defaultValue=0.5, parameterName=DesaturationFraction})
  n23 = ComponentMask(None=@n2.XYZ Texture) {g=true}
  n24 = Add(A=Multiply(A=ConstantBiasScale(None=@n3) {bias=2}, B=Multiply(A=@n20, B=@n22)), B=Multiply(A=@n5, B=OneMinus(None=@n23)))
  Add(A=Multiply(A=@n24, B=VectorParameter() {parameterName=Color, defaultValue=<rgba 0.0173 0.1038 0.3 0>}.RGB), B=@n21)
  Add(B=@n24)
  n25 = Add(A=SceneTexture() {sceneTextureId=PPI_CustomDepth}.Color, B=ScalarParameter() {defaultValue=10, parameterName=Custom Depth Offset Amount})
  n26 = SceneTexture() {sceneTextureId=PPI_SceneDepth}
  ComponentMask(None=Subtract(A=@n25, B=@n26.Color)) {r=true}
  n27 = Multiply(A=ComponentMask(None=CollectionParameter() {collection=&MPC_Interrogate, parameterName=OriginPositionWS}) {b=true, g=true, r=true})
  n28 = Multiply(A=WorldPosition().XYZ, B=Constant3Vector() {constant=<rgba 1 1 0 0>})
  n29 = MaterialFunctionCall(
   TimestampA=CollectionParameter() {collection=&MPC_Interrogate, parameterName=ActivatedTimestamp}
   TransitionLength_In=CollectionParameter() {collection=&MPC_Interrogate, parameterName=Duration}
   Transition In Only?=StaticBool() {value=true}
  ) {materialFunction=&MF_Timestamp_InOut_01}
  n30 = Multiply(
   A=CollectionParameter() {collection=&MPC_Interrogate, parameterName=Radius}
   B=@n29.Linear
  )
  n31 = SphereMask(A=@n27, B=@n28, Radius=@n30, Hardness=Constant() {r=0.5})
  n32 = Constant()
  Multiply(A=@n31, B=OneMinus(None=SphereMask(A=@n27, B=@n28, Radius=@n30, Hardness=@n32)))
  n33 = CollectionParameter() {collection=&MPC_Interrogate, parameterName=StencilDuration}
  Multiply(A=Clamp(None=@n31), B=Clamp(None=LinearInterpolate(A=@n33, Alpha=@n29.Linear) {constA=1, constB=0}))
  n34 = MaterialFunctionCall!(ID=CollectionParameter() {collection=/Game/Packages/Fortress_SharedMaterials/Stencil_ID_Definitions/Stencil_ID_Definitions, parameterName=ScooterRedOverlay}) {materialFunction=/Game/Packages/Fortress_SharedMaterials/Stencil_ID_Definitions/StencilMaskCompare}
  n35 = Subtract(A=@n26.Color, B=@n25)
  n36 = OneMinus(None=Ceil(None=Clamp(None=@n35)) {desc=z tested objects black z tested - everything else white}) {desc=occlusion mask}
  n37 = Sine(None=Multiply(A=Time(), B=ScalarParameter() {defaultValue=0.75, parameterName=PulseRate}))
  n38 = ComponentMask(None=ScreenPosition().ViewportUV) {g=true, r=true}
  n39 = Subtract(A=Divide(A=ViewSize(), B=Constant() {r=320}), B=Constant() {r=2}) {constB=0}
  n40 = Multiply(A=Clamp(None=Ceil(None=@n39)) {maxDefault=2, minDefault=1}, B=SceneTexelSize()) {constA=2}
  n41 = Constant()
  n42 = AppendVector(A=ComponentMask(None=@n40) {desc=x, r=true}, B=@n41)
  n43 = Add(A=@n38, B=@n42) {desc=x+}
  n44 = AppendVector(A=@n41, B=ComponentMask(None=@n40) {desc=y, g=true}) {desc=y+}
  n45 = Add(A=@n38, B=@n44) {desc=y+}
  n46 = Constant() {r=-1}
  n47 = Add(A=@n38, B=Multiply(A=@n42, B=@n46)) {desc=x-}
  n48 = Add(A=@n38, B=Multiply(A=@n46, B=@n44)) {desc=y-}
  n49 = ComponentMask(None=MaterialFunctionCall!(
    Base=SceneTexture() {sceneTextureId=PPI_CustomDepth}.Color
    BaseSampleBias
    Scale
    Sample1=SceneTexture(UVs=@n43) {sceneTextureId=PPI_CustomDepth}.Color
    Sample2=SceneTexture(UVs=@n45) {sceneTextureId=PPI_CustomDepth}.Color
    Sample3=SceneTexture(UVs=@n47) {sceneTextureId=PPI_CustomDepth}.Color
    Sample4=SceneTexture(UVs=@n48) {sceneTextureId=PPI_CustomDepth}.Color
    Divisor=Constant() {r=-16}
    UseCenteredLines
    UseRGB
   ) {materialFunction=/Game/Packages/jl_tests/combine4Samples}.Result) {r=true}
  n50 = Add(A=LinearInterpolate(Alpha=ConstantBiasScale(None=@n37)) {constA=0.005, constB=0.02}, B=@n49)
  n51 = Clamp(None=@n36)
  n52 = LinearInterpolate(Alpha=@n49) {constA=0.2}
  n53 = Multiply(A=OneMinus(None=@n1.Linear))
  n54 = Multiply(A=Multiply(A=@n5, B=Saturate(None=@n53))) {constB=3}
  n55 = Multiply(A=Add(A=Multiply(A=@n36, B=@n50), B=Multiply(A=@n51, B=@n52)), B=Clamp(None=Multiply(A=@n54, B=@n51)) {maxDefault=100, minDefault=0.1})
  n56 = LinearInterpolate(
   B=VectorParameter() {parameterName=EnemyColor, defaultValue=<rgba 1.5 0 0 0>}.RGB
   Alpha=ComponentMask(None=Multiply(A=@n34.Result, B=@n55)) {r=true}
  )
  n57 = Constant3Vector() {constant=<rgba 1 0.7085 0 0>}
  n58 = CollectionParameter() {collection=&Stencil_ID_Definitions, parameterName=ScooterYellowOverlay}
  n59 = ComponentMask(None=Multiply(
    A=MaterialFunctionCall(ID=@n58) {materialFunction=&StencilMaskCompare}.Result
    B=@n55
   )) {r=true}
  n60 = Add(A=@n21, B=Multiply(A=LinearInterpolate(A=@n56, B=@n57, Alpha=@n59)))
  ShadingPathSwitch(Default=@n60, Deferred=@n60, Forward=@n60, Mobile=@n60)
 ]
}
