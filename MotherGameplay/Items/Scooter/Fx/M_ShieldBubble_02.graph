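# Additive unlit material for a scooter shield bubble effect, usable on Niagara mesh particles.
# Tints clamped scene color by time of day and depth, with a flat-color fallback for forward and mobile paths.
# Layers sine-warped normal-map caustics, distance-faded scanlines and grid, and a depth-intersection edge, multiplied by particle color.
# The same panning normals build vertex ripple offsets, while caustic-driven erosion shapes particle alpha and dynamic-parameter refraction.
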Material /MotherGameplay/Items/Scooter/Fx/M_ShieldBubble_02
graph {
 properties {
  bEnableResponsiveAA=true
  bUseTranslucencyVertexFog=false
  bUsedWithNiagaraMeshParticles=true
  blendMode=BLEND_Additive
  pixelDepthOffsetMode=PDOM_Legacy
  refractionMethod=RM_PixelNormalOffset
  shadingModel=MSM_Unlit
  translucencyPass=MTP_BeforeDOF
  usedShadingModels=Unlit
 }
 nodes [
  n1 = MaterialFunctionCall!(Distance Bias, Distance Divisor=ScalarParameter() {defaultValue=32, parameterName=Intersection Edge Divisor}, Scene Depth, Pixel Depth) {materialFunction=/Game/Effects/Fort_Effects/Materials/_Functions/MF_DistanceBehindTranslucency_01}
  n2 = Power(Base=@n1.Normalized Inverted, Exp=ScalarParameter() {defaultValue=6, parameterName=Intersection Edge Exponent})
  n3 = AppendVector(
   A=ScalarParameter() {defaultValue=2, parameterName=U Repeat Caustic 1}
   B=ScalarParameter() {defaultValue=1.232, parameterName=V Repeat Caustic 1}
  )
  n4 = MaterialFunctionCall() {materialFunction=/Engine/Functions/Engine_MaterialFunctions02/WorldPositionOffset/ObjectScale}
  n5 = Multiply(A=Divide(A=Round(None=@n4.Scale X), B=ScalarParameter() {defaultValue=2, parameterName=Tile UV Scale}), B=TextureCoordinate())
  n6 = TextureSample(UVs=Panner(Coordinate=Multiply(A=@n5, B=@n3)) {bFractionalPart=true, speedX=0.0268, speedY=-0.0194}) {samplerType=SAMPLERTYPE_Normal, texture=/Game/TheOutpost/Textures/TilingWeakPointNormal}
  n7 = Time() {bOverride_Period=true, period=7.5075}
  n8 = ScalarParameter() {defaultValue=0.1332, parameterName=Sine Speed A}
  n9 = Add(A=@n6.RGB, B=ConstantBiasScale(None=Multiply(A=@n7, B=@n8)) {bias=-0.5, scale=2.2})
  n10 = Time() {bOverride_Period=true, period=15.9185}
  n11 = ScalarParameter() {defaultValue=0.0628, parameterName=Sine Speed B}
  n12 = AppendVector(
   A=ScalarParameter() {defaultValue=1, parameterName=U Repeat Caustic 2}
   B=ScalarParameter() {defaultValue=0.5312, parameterName=V Repeat Caustic 2}
  )
  n13 = TextureSample(UVs=Panner(Coordinate=Multiply(A=@n5, B=@n12)) {bFractionalPart=true, speedX=-0.0226, speedY=-0.0104}) {samplerType=SAMPLERTYPE_Normal, texture=&TilingWeakPointNormal}
  n14 = Add(A=ConstantBiasScale(None=Multiply(A=@n10, B=@n11)) {bias=-0.5, scale=4.4}, B=@n13.RGB)
  n15 = Desaturation(None=Multiply(A=Sine(None=@n9) {period=2.2}, B=Sine(None=@n14) {period=4.4})) {luminanceFactors=<rgba 1 1 1 1>}
  n16 = OneMinus(None=Abs(None=@n15))
  n17 = Power(Base=Saturate(None=@n16), Exp=ScalarParameter() {defaultValue=5, parameterName=Caustic Exponent})
  n18 = Round(None=@n17)
  Add(A=@n2, B=Multiply(A=@n1.Normalized Inverted, B=@n18))
  n19 = AppendVector(A=ScalarParameter() {defaultValue=63, parameterName=U Raindrops}, B=ScalarParameter() {defaultValue=25, parameterName=V Raindrops})
  n20 = TextureSample(UVs=Panner(Coordinate=Multiply(A=@n5, B=@n19)) {bFractionalPart=true, speedY=-0.1}) {samplerType=SAMPLERTYPE_LinearColor, texture=/Game/TheOutpost/Textures/T_RainDropGradient_01}
  n21 = AppendVector(A=ScalarParameter() {defaultValue=37, parameterName=U Raindrops Mask}, B=ScalarParameter() {defaultValue=16, parameterName=V Raindrops Mask})
  n22 = TextureSample(UVs=Panner(Coordinate=Multiply(A=@n5, B=@n21)) {bFractionalPart=true, speedX=0.324, speedY=-0.187}) {texture=&T_NoisePack_01}
  n23 = Time() {bOverride_Period=true, period=0.625}
  n24 = ScalarParameter() {defaultValue=1.6, parameterName=Raindrop Speed}
  n25 = Sine(None=Add(A=Add(A=@n20.G, B=@n22.R), B=Multiply(A=@n23, B=@n24))) {period=1.6}
  n26 = Power(Base=Saturate(None=Multiply(A=@n20.R, B=@n25)), Exp=ScalarParameter() {defaultValue=8, parameterName=Rain Exponent})
  Add(B=Multiply(A=Multiply(A=@n26, B=@n22.B), B=VectorParameter() {parameterName=Rain Color, defaultValue=<rgba 0.25 1.5 20 1>}.RGB))
  n27 = ScalarParameter() {defaultValue=1.02, parameterName=RefractionAmount}
  n28 = DynamicParameter() {paramNames [refraction,Param2,Param3,Param4]}
  n29 = ScalarParameter() {parameterName=Refraction Bias}
  n30 = ScalarParameter() {defaultValue=500, parameterName=Refraction Divisor}
  n31 = ScalarParameter() {defaultValue=1, parameterName=Erode_Exponent}
  n32 = OneMinus(None=ConstantBiasScale(None=@n15) {bias=2, scale=0.25})
  n33 = Add(A=Clamp(None=Power(Base=@n32, Exp=@n31) {constExponent=1}), B=ScalarParameter() {defaultValue=1, parameterName=Erode})
  LinearInterpolate(
   B=LinearInterpolate(
    B=Multiply(A=@n27, B=@n28.refraction)
    Alpha=MaterialFunctionCall(Distance Bias=@n29, Distance Divisor=@n30) {materialFunction=&MF_DistanceBehindTranslucency_01}.Normalized
   ) {constA=1}
   Alpha=Saturate(None=@n33)
  ) {constA=1}
  n34 = Panner(Coordinate=@n5) {bFractionalPart=true, speedY=-0.005}
  n35 = ScalarParameter() {defaultValue=200, parameterName=Line Repeat}
  n36 = ComponentMask(None=Frac(None=Multiply(A=@n34, B=@n35))) {g=true}
  n37 = Add(A=Multiply(A=@n36, B=OneMinus(None=@n36)), B=ScalarParameter() {defaultValue=-0.247, parameterName=Line Ceil Bias})
  n38 = Multiply(A=Ceil(None=Saturate(None=@n37)), B=VectorParameter() {parameterName=Lines Color, defaultValue=<rgba 0.0928 0.2583 0.8 1>}.RGB)
  n39 = MaterialFunctionCall(
   Position=WorldPosition().XYZ
   Fade_Start_Distance=ScalarParameter() {defaultValue=4000, parameterName=Fade Distance Start}
   Fade_End_Distance=ScalarParameter() {defaultValue=2000, parameterName=Fade Distance End}
  ) {materialFunction=&MF_Distance_SyncedFade_FX}
  n40 = ComponentMask(None=TextureCoordinate()) {g=true}
  n41 = ScalarParameter() {defaultValue=0.7, parameterName=BottomTopBias}
  n42 = Saturate(None=Power(Base=Add(A=@n40, B=@n41), Exp=ScalarParameter() {defaultValue=15, parameterName=BottomTopExponent}))
  n43 = Clamp(None=SceneColor!(UVs).RGB, Max=ScalarParameter() {defaultValue=0.15, desc=Avoid super bloom, parameterName=Scene Color Clamp Max}) {maxDefault=5}
  n44 = MaterialFunctionCall(
   Distance Bias=ScalarParameter() {defaultValue=500, parameterName=Saturation Distance Bias}
   Distance Divisor=ScalarParameter() {defaultValue=10000, parameterName=Saturation Divisor}
  ) {materialFunction=&MF_DistanceBehindTranslucency_01}
  n45 = MaterialFunctionCall(
   Color_Morning=LinearInterpolate(
    A=VectorParameter() {parameterName=Morning Color Near, defaultValue=<rgba 5.5236 46.1237 100 1>}.RGB
    B=VectorParameter() {parameterName=Morning Color Far, defaultValue=<rgba 1.8346 16.8598 25 1>}.RGB
    Alpha=@n44.Normalized
   )
   Color_Day=LinearInterpolate(
    A=VectorParameter() {parameterName=Day Color Near, defaultValue=<rgba 20 37.3333 100 1>}.RGB
    B=VectorParameter() {parameterName=Day Color Far, defaultValue=<rgba 2.5 6.25 25 1>}.RGB
    Alpha=@n44.Normalized
   )
   Color_Evening=LinearInterpolate(
    A=VectorParameter() {parameterName=Evening Color Near, defaultValue=<rgba 9 39.5499 150 1>}.RGB
    B=VectorParameter() {parameterName=Evening Color Far, defaultValue=<rgba 2.1939 26.6086 50 1>}.RGB
    Alpha=@n44.Normalized
   )
   Color_Night=LinearInterpolate(
    A=VectorParameter() {parameterName=Night Color Near, defaultValue=<rgba 11.1809 32.9473 75 1>}.RGB
    B=VectorParameter() {parameterName=Night Color Far, defaultValue=<rgba 3.0677 9.4303 25 1>}.RGB
    Alpha=@n44.Normalized
   )
  ) {materialFunction=&Color_Time_of_Day}
  n46 = Normalize(VectorInput=@n45.Result)
  n47 = Power(
   Base=Add(A=Multiply(A=@n43, B=@n46), B=ScalarParameter() {defaultValue=0.11, parameterName=Scene Color Bias})
   Exp=ScalarParameter() {defaultValue=2.8, parameterName=Scene Color Contrast}
  )
  n48 = LinearInterpolate(A=@n47, B=Desaturation(None=@n47) {luminanceFactors=<rgba 1 1 1 0.25>}, Alpha=@n44.Normalized)
  n49 = VectorParameter() {parameterName=PerformanceMode_Color, defaultValue=<rgba 0 0.0901 0.7083 0>}
  n50 = LinearInterpolate(A=@n17, B=@n18)
  n51 = Add(
   A=ShadingPathSwitch(Default=Multiply(A=@n48, B=@n45.Result), Forward=@n49.RGB, Mobile=@n49.RGB)
   B=Multiply(A=VectorParameter() {parameterName=Color Caustics, defaultValue=<rgba 0.0064 0.0474 0.16 1>}.RGB, B=Multiply(A=@n50, B=@n42))
  ) {desc=+ caustics}
  n52 = ScalarParameter() {defaultValue=43.5, parameterName=U Repeat Grid}
  n53 = ScalarParameter() {defaultValue=13, parameterName=V Repeat Grid}
  n54 = TextureSample(UVs=Multiply(A=@n5, B=AppendVector(A=@n52, B=@n53))) {samplerType=SAMPLERTYPE_LinearColor, texture=/Game/Gadgets/Assets/Teleporter/Textures/T_Teleporter_GridSquares_01}
  n55 = Multiply(A=@n16, B=ScalarParameter() {defaultValue=0.2, parameterName=Ripple to Grid Scale})
  n56 = Add(
   A=@n54.R
   B=Multiply(A=Time() {bOverride_Period=true, period=3.8314}, B=ScalarParameter() {defaultValue=0.261, parameterName=Sine Speed Grid})
  )
  n57 = Multiply(A=@n54.G, B=Sine(None=Add(A=@n55, B=@n56)))
  n58 = ScalarParameter() {defaultValue=5, parameterName=Grid Power}
  n59 = Multiply(
   A=Multiply(A=@n39.Result Inverted, B=Power(Base=@n57, Exp=@n58))
   B=VectorParameter() {parameterName=Grid Color, defaultValue=<rgba 0.0143 0.063 0.15 0>}.RGB
  )
  n60 = Add(
   A=Multiply(A=Multiply(A=@n38, B=@n39.Result Inverted) {desc=lines * distance fade}, B=@n42) {desc=lines * gradient}
   B=Add(A=@n51, B=Multiply(A=@n42, B=@n59)) {desc=+ Grid Blocks}
  ) {desc=+ lines}
  n61 = Multiply(A=@n2, B=VectorParameter() {parameterName=Intersection Edge Color, defaultValue=<rgba 0.25 1.5 20 1>}.RGB)
  n62 = ParticleColor()
  MaterialFunctionCall(Emissive=Multiply(A=Add(A=@n60, B=@n61) {desc=+ intersection edge}, B=@n62.RGB)) {materialFunction=&MF_FX_EmissiveAdjustments}
  n63 = Add(A=@n6.RGB, B=@n13.RGB)
  n64 = Multiply(A=Multiply(A=@n63, B=VertexNormalWS()), B=VectorParameter() {parameterName=WPO Strength, defaultValue=<rgba 18 18 18 0>}.RGB)
  Multiply(A=Multiply(A=Multiply(A=@n42, B=@n64), B=@n39.Result Inverted), B=VertexColor().B)
  n65 = Multiply(B=ScalarParameter() {defaultValue=0.75, parameterName=Flicker Speed})
  n66 = Add(A=Sine(None=@n65) {period=1.0263}, B=Sine(None=@n65) {period=0.5197})
  Multiply(
   A=ScalarParameter() {defaultValue=0.0015, parameterName=Flicker Intensity}
   B=Power(Base=Clamp(None=@n66) {clampMode=CMODE_ClampMin}, Exp=ScalarParameter() {defaultValue=2, parameterName=Flicker Random Time Exponent}) {constExponent=6}
  )
  Multiply(A=@n33, B=@n62.A)
  n67 = TextureCoordinate()
  n68 = AppendVector(A=ScalarParameter() {defaultValue=750, parameterName=Noise U}, B=ScalarParameter() {defaultValue=250, parameterName=Noise V})
  Multiply(
   A=TextureSample(UVs=Multiply(A=@n67, B=@n68)) {samplerType=SAMPLERTYPE_LinearColor, texture=/Game/Global/Textures/Default/Noise/Noise32x32}.R
   B=ScalarParameter() {defaultValue=0.1, parameterName=Noise Scale}
  )
  QualitySwitch(Default=@n60, Low=@n51, High=@n60, Medium=@n51)
  Reroute(None=Normalize(VectorInput=@n63))
 ]
}
