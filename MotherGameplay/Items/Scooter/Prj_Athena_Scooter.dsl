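; Projectile for the Scooter recon item, which on impact can become a physics-driven pulse indicator actor.
; The server attaches a separate FX actor at spawn, while non-server instances schedule near-death effects at 75% of lifetime.
; Bounces save velocity and impact angle; if the data toggle is on, the projectile stops, spawns the pulse actor with an angle-scaled impulse, hands over weapon and FX, then destroys itself.
; Water entry and non-physics destruction trigger GameplayCues.
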(event Collision|EventActorBeginOverlap
  (params (OtherActor :type Actor%20Object%20Reference)))

(event EventTick
  (params (DeltaSeconds :type Float%20%28single-precision%29)))

(event Custom|NearDeathFx
  (Components|Activation|Activate (Variables|Default|GetNearDeathBlink) false)
  (Audio|SpawnSoundAttached "/MotherGameplay/Sounds/ScooterRecon/ScooterRecon_Projectile_NearDeath_Cue.ScooterRecon_Projectile_NearDeath_Cue" (Variables|Default|GetInAirAudioLoop)))

(event EventBeginPlay
  (bind _duration (Variables|Default|GetDuration))
  (bind _returnvalue (ScalableFloat|GetValueatLevel _duration))
  ;[1 1 1] Spawn actor that holds FX. bind it's destruction to this actors destruction
  (|CallParentFunction)
  (Actor|SetLifeSpan _returnvalue)
  (if (Networking|IsServer)
    (bind _returnvalue_2 (Game|SpawnActorfromClass "/ScooterGameplay/Items/Scooter/BP_Mother_Scooter_FX.BP_Mother_Scooter_FX_C" (Math|Transform|MakeTransform (.location (Transformation|GetActorTransform self)) (.rotation (Transformation|GetActorTransform self))) "AlwaysSpawn" "OverrideRootScale" 0 (Game|GetInstigator)))
    (Transformation|AttachActorToActor _returnvalue_2 self "None" "SnapToTarget" "SnapToTarget" "KeepWorld" true)
    (Game|BindEventtoOnDestroyed self (delegate Custom|CustomEvent_0))
    (if (not (Class|AthenaFunctionLibrary|RowtoBool (Variables|Default|GetSpawnPhysicsObject)))
      (Variables|/Script/Engine.ProjectileMovementComponent|SetBounceVelocityStopSimulatingThreshold 25.0 (Variables|Movement|GetProjectileMovementComponent)))
    (else
      (Utilities|Time|SetTimerbyEvent (delegate Custom|NearDeathFx) (* _returnvalue 0.75)))))

(event Custom|CustomEvent_0
  (params (DestroyedActor :type Actor%20Object%20Reference))
  ;[1 1 1] Spawn actor that holds FX. bind it's destruction to this actors destruction
  (Actor|DestroyActor (stale EventBeginPlay _returnvalue_2)))

(event OnProjectileStop(ProjectileMovementComponent)
  (params (ImpactResult :type Hit%20Result%20Structure%20%28by%20ref%29))
  (bind _returnvalue (Actor|GetLifeSpan))
  ;[1 1 1] Spawn Physics Actor
  (if (Class|AthenaFunctionLibrary|RowtoBool (Variables|Default|GetSpawnPhysicsObject))
    (bind _returnvalue_2 (Building|SpawnBuildingGameplayActor "/MotherGameplay/Items/Scooter/BGA_Athena_Scooter_Child.BGA_Athena_Scooter_Child_C" (Math|Transform|MakeTransform (.location (Transformation|GetActorTransform)) (Math|Rotator|CombineRotators (.rotation (Transformation|GetActorTransform)) "0,90.000000, 0")) (Variables|Actor|GetInstigator)))
    (if (!= _returnvalue 0.0)
      (Actor|SetLifeSpan _returnvalue_2 _returnvalue)
      (else
        _))
    ;[1 1 1] Unbind fx actor destruction (it gets rebinded to the new physics actor)
    (Game|UnbindEventfromOnDestroyed self (delegate Custom|CustomEvent_0))
    (bind _asbga_pulse_indicator (Utilities|Casting|CastToBGA_PulseIndicator _returnvalue_2))
    (Development|PrintString (Utilities|String|ToString(Float) (Variables|Default|GetSavedDot)))
    (bind _returnvalue_3 (FortPhysicsSimulationLibrary|IsPhysicsObjectActiveinSimulation _asbga_pulse_indicator))
    (if _returnvalue_3
      (FortPhysicsSimulationLibrary|AddPhysicsImpulse _asbga_pulse_indicator (* (Variables|Default|GetSavedVelocity) (Math|Float|Lerp (ScalableFloat|GetValueatLevel (Variables|Default|GetMaxBounceVelocity)) (ScalableFloat|GetValueatLevel (Variables|Default|GetMinBounceVelocity)) (Variables|Default|GetSavedDot))) true "None" false false 0 "(MaxSimSizeAllowed=Large,bRejectEmbeddedInLandscape=False)")
      (else
        _))
    ;[1 1 1] Pass Owning Weapon
    (bind _returnvalue_4 (CallFunction|GetOwnerWeapon))
    (Variables|/Game/Athena/Items/Consumables/PulseIndicator/BGA_PulseIndicator.BGA_PulseIndicator_C|SetOwnerWeapon _returnvalue_4 _asbga_pulse_indicator)
    (Class|BGAPulseIndicator|StartPulseMarkers _asbga_pulse_indicator)
    (Transformation|AttachActorToActor (stale EventBeginPlay _returnvalue_2) _asbga_pulse_indicator "None" "SnapToTarget" "SnapToTarget" "KeepWorld" true)
    (Class|BGAPulseIndicator|Ready _asbga_pulse_indicator)
    (Actor|DestroyActor)))

(event EventOnBounce
  (params (Hit :type Hit%20Result%20Structure%20%28by%20ref%29))
  (bind (_bblockinghit _binitialoverlap _time _distance _location _impactpoint _normal _impactnormal) (Collision|BreakHitResult Hit))
  ;[1 1 1] Spawn Physics Actor
  (bind _output_get (Variables|Default|SetSavedDot (Math|Vector|DotProduct _impactnormal (Math|Vector|Normalize (Transformation|GetVelocity)))))
  (Development|PrintString (Utilities|String|ToString(Float) _output_get) true true "(R=0.000000,G=1.000000,B=0.004274,A=1.000000)" 5.0)
  (bind _output_get_2 (Variables|Default|SetSavedVelocity (Variables|/Script/Engine.MovementComponent|GetVelocity (Variables|Movement|GetProjectileMovementComponent))))
  (Development|PrintString (Utilities|String|ToString(Vector) _output_get_2) true true "(R=1.000000,G=0.835607,B=0.000000,A=1.000000)" 5.0)
  (Variables|/Script/Engine.ProjectileMovementComponent|SetProjectileGravityScale 0.3 (Variables|Movement|GetProjectileMovementComponent))
  (if (Class|AthenaFunctionLibrary|RowtoBool (Variables|Default|GetSpawnPhysicsObject))
    (Game|Components|ProjectileMovement|StopSimulating (Variables|Movement|GetProjectileMovementComponent) Hit)))

(event OnEnterWaterEvent(WaterInteractionComponent)
  (params (WaterBody :type Fort%20Water%20Body%20Actor%20Object%20Reference) (WaterInteractionComponent :type Fort%20Water%20Interaction%20Component%20Object%20Reference) (bIsFirstBody :type Boolean))
  ;[1 1 1] Spawn Physics Actor
  (GameplayCue|ExecuteGameplayCue%28Burst%29Local self (Variables|Default|GetWater%20Cue) (Ability|GameplayCue|MakeGameplayCueParameters 0.0)))

(event EventDestroyed
  ;[1 1 1] Spawn Physics Actor
  (if (not (Class|AthenaFunctionLibrary|RowtoBool (Variables|Default|GetSpawnPhysicsObject)))
    (GameplayCue|ExecuteGameplayCue%28Burst%29Local self (Variables|Default|GetDestroyedCue) (Ability|GameplayCue|MakeGameplayCueParameters 0.0))))

(fn ConstructionScript
  (|CallParentFunction))
